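# Test Control Rig for the PistolCanary creature's eyed tentacle, prototyping ground-trace IK near the tip.
# Each update snaps ik_tentacle to tentacle_14, then sphere-traces vertically there with a debug line.
# Hit Z becomes ZOffset_Target, smoothed into ZOffset by AlphaInterp; misses zero both.
# CCDIK bends the tentacle toward ik_tentacle; chained after the solve are debug prints and the ZOffset modification of ik_tentacle; DoIKTrace is unused.
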ControlRigBlueprint /PistolCanary/Anim/Test/ABP_TraceTesting/Tentacle_PistolCanary_Creature_IKTrace_CtrlRig
graphs [
 New Function_ContainedGraph {nodes [
   n1 = RigVMFunctionEntryNode()
   n2 = &RigUnit_GetTransform(bInitial=False, Item=@n1.Item)
   n3 = &RigUnit_GetTransform(Item=(Type=Bone,Name="tentacle_14"), bInitial=False)
   Add = /Script/RigVM.RigVMFunction_MathVectorAdd(
    A=(X=0,Y=0,Z=0)
    B=(X=0,Y=0,Z=2500)
    B.Z=@n1.Trace Max Height
    A.X=@n2.Transform.Translation.X
    A.Y=@n2.Transform.Translation.Y
    A.Z=@n3.Transform.Translation.Z
   ) {is=struct RigVMUnitNode, template="Add::Execute(in A,in B,out Result)"}
   Subtract = /Script/RigVM.RigVMFunction_MathVectorSub(
    A=(X=0,Y=0,Z=0)
    B=(X=0,Y=0,Z=2500)
    B.Z=@n1.Trace Min Height
    A.X=@n2.Transform.Translation.X
    A.Y=@n2.Transform.Translation.Y
    A.Z=@n3.Transform.Translation.Z
   ) {is=struct RigVMUnitNode, template="Subtract::Execute(in A,in B,out Result)"}
   > RigVMFunction_DebugLineNoSpace = /Script/RigVM.RigVMFunction_DebugLineNoSpace(
    Color=(R=1,G=0,B=0,A=1)
    DebugDrawSettings=(DepthPriority=SDPG_Foreground,Lifetime=-1)
    Thickness=100
    WorldOffset=(Rotation=(X=0,Y=0,Z=0,W=1),Translation=(X=0,Y=0,Z=0),Scale3D=(X=1,Y=1,Z=1))
    bEnabled=true
    A=@Add.Result
    B=@Subtract.Result
   ) {is=struct RigVMUnitNode}
   n4 = /Script/ControlRig.RigUnit_SphereTraceByTraceChannel(Radius=10, TraceChannel=TraceTypeQuery1, Start=@Add.Result, End=@Subtract.Result) {is=struct RigVMUnitNode}
   > RigVMFunctionReturnNode(Hit Location=@n4.HitLocation, Hit=@n4.bHit)
  ]}
 RigVMFunctionLibrary {nodes [ground_ik_trace = RigVMCollapseNode(ExecuteContext=(), Item=(Type=Bone,Name="None"), Trace Max Height=2500, Trace Min Height=2500)]}
 RigVMFunction_Sequence_1_ContainedGraph {nodes [
   n1 = RigVMFunctionEntryNode()
   RigVMFunction_Sequence = &RigVMFunction_Sequence(ExecuteContext=@n1.ExecuteContext)
   n2 = &RigVMFunction_Sequence(ExecuteContext=@RigVMFunction_Sequence.B)
   n3 = &RigVMFunction_Sequence(ExecuteContext=@n2.B)
   RigVMFunctionReturnNode(A=@RigVMFunction_Sequence.A, B=@n2.A, C=@n3.A, D=@n3.B)
  ]}
 RigVMModel {nodes [
   RigUnit_BeginExecution = &RigUnit_BeginExecution()
   RigVMFunction_Sequence = RigVMAggregateNode(ExecuteContext=@RigUnit_BeginExecution.ExecutePin)
   ground_ik_trace = RigVMFunctionReferenceNode(Item=(Type=Bone,Name="tentacle_14"), Trace Max Height=3500, Trace Min Height=3500, ExecuteContext=@RigVMFunction_Sequence.B)
   RigVMFunction_ControlFlowBranch = &RigVMFunction_ControlFlowBranch(ExecuteContext=@ground_ik_trace.ExecuteContext, Condition=@ground_ik_trace.Hit)
   RigVMVariableNode(ExecuteContext=@RigVMFunction_ControlFlowBranch.False) {variable=ZOffset}
   > RigVMVariableNode() {variable=ZOffset_Target}
   RigVMVariableNode(ExecuteContext=@RigVMFunction_ControlFlowBranch.True, Value=@ground_ik_trace.Hit Location.Z) {variable=ZOffset_Target}
   RigUnit_GetTransform = &RigUnit_GetTransform(Item=(Type=Bone,Name="ik_tentacle"), bInitial=False)
   RigUnit_ItemArray = &RigUnit_ItemArray(Items="((Type=Bone,Name=\"tentacle_01\"),(Type=Bone,Name=\"tentacle_02\"),(Type=Bone,Name=\"tentacle_03\"),(Type=Bone,Name=\"tentacle_04\"),(Type=Bone,Name=\"tentacle_05\"),(Type=Bone,Name=\"tentacle_06\"),(Type=Bone,Name=\"tentacle_07\"),(Type=Bone,Name=\"tentacle_08\"),(Type=B ...(truncated)")
   /Script/ControlRig.RigUnit_CCDIKItemArray(
    BaseRotationLimit=10
    MaxIterations=10
    Precision=1
    RotationLimits=()
    Weight=1
    bPropagateToChildren=True
    bStartFromTail=True
    EffectorTransform=@RigUnit_GetTransform.Transform
    Items=@RigUnit_ItemArray.Items
    ExecutePin=@RigVMFunction_Sequence.D
   ) {is=struct RigVMUnitNode}
   > RigVMDispatchNode(Enabled=true, Prefix=ZOffset, ScreenColor=(R=1,G=1,B=1,A=1), ScreenDuration=0.05, Value=RigVMVariableNode() {variable=ZOffset}.Value) {template="DISPATCH_RigVMDispatch_Print(in Prefix,in Value,in Enabled,in ScreenDuration,in ScreenColor)"}
   > RigVMDispatchNode(
    Enabled=true
    Prefix=ZOffset_Target
    ScreenColor=(R=1,G=1,B=1,A=1)
    ScreenDuration=0.05
    Value=RigVMVariableNode() {variable=ZOffset_Target}.Value
   ) {template="DISPATCH_RigVMDispatch_Print(in Prefix,in Value,in Enabled,in ScreenDuration,in ScreenColor)"}
   > &RigUnit_ModifyTransforms(
    ItemToModify=((Item=(Type=Bone,Name="ik_tentacle"),Transform=(Rotation=(X=0,Y=0,Z=0,W=1),Translation=(X=0,Y=0,Z=0),Scale3D=(X=1,Y=1,Z=1))))
    Weight=1
    WeightMinimum=0
    ItemToModify.0.Transform.Translation.Z=RigVMVariableNode() {variable=ZOffset}.Value
   )
   AlphaInterp = &RigVMFunction_AlphaInterp(InterpSpeedDecreasing=15, InterpSpeedIncreasing=15, Scale=1, bClampResult=False, Value=RigVMVariableNode() {variable=ZOffset_Target}.Value)
   RigVMVariableNode(Value=@AlphaInterp.Result, ExecuteContext=@RigVMFunction_Sequence.C) {variable=ZOffset}
   RigUnit_SetTransform = &RigUnit_SetTransform(
    Item=(Type=Bone,Name="ik_tentacle")
    Weight=1
    bInitial=False
    Value=&RigUnit_GetTransform(Item=(Type=Bone,Name="tentacle_14"), bInitial=False).Transform
    ExecutePin=@RigVMFunction_Sequence.A
   )
  ]}
]
hierarchy [
 root {loc=<0 0 0>, rot=<0 0 0 1>}
 tentacle_base {parent=root, loc=<0.2999 -120.5468 -1636.4241>, rot=<-0.5 0.5 -0.5 -0.5>}
 tentacle_01 {parent=tentacle_base, loc=<1269.9846 -0.7843 -19.493>, rot=<0 -0.0043 0.0001 -1>}
 tentacle_02 {parent=tentacle_01, loc=<1279.0715 0 0>, rot=<0 0.0268 -0.0003 -0.9996>}
 tentacle_03 {parent=tentacle_02, loc=<1275.3529 0 0>, rot=<0 0.0031 0.0001 -1>}
 tentacle_04 {parent=tentacle_03, loc=<1119.8098 0 0>, rot=<0 -0.0112 0.0005 -0.9999>}
 tentacle_05 {parent=tentacle_04, loc=<1114.6722 0 0>, rot=<0 -0.0006 -0.0006 -1>}
 tentacle_06 {parent=tentacle_05, loc=<1126.2786 0 0>, rot=<0 0.0116 0.0009 -0.9999>}
 tentacle_07 {parent=tentacle_06, loc=<1115.1567 0 0>, rot=<0 -0.0294 -0.0008 -0.9996>}
 tentacle_08 {parent=tentacle_07, loc=<1113.4962 0 0>, rot=<0 0.0056 0.0001 -1>}
 tentacle_09 {parent=tentacle_08, loc=<1119.848 0 0>, rot=<0 -0.0065 0.0007 -1>}
 tentacle_10 {parent=tentacle_09, loc=<1112.4393 0 0>, rot=<0 -0.0094 -0.0007 -1>}
 tentacle_11 {parent=tentacle_10, loc=<1019.1152 0 0>, rot=<0 -0.0039 0.0004 -1>}
 tentacle_12 {parent=tentacle_11, loc=<696.9656 0 0>, rot=<0 0.0047 -0.0001 -1>}
 tentacle_13 {parent=tentacle_12, loc=<658.0217 0 0>, rot=<0 -0.0077 -0.0002 -1>}
 tentacle_14 {parent=tentacle_13, loc=<599.0422 0 0>, rot=<0 -0.0032 -0.0001 -1>}
 tentacle_15 {parent=tentacle_14, loc=<657.8021 0 0>, rot=<0 -0.0128 0 -0.9999>}
 tentacle_16 {parent=tentacle_15, loc=<541.2894 0 0>, rot=<0 -0.0132 0.0001 -0.9999>}
 tentacle_17 {parent=tentacle_16, loc=<466.9173 0 0>, rot=<0 -0.0119 0.0001 -0.9999>}
 eyeE {parent=tentacle_10, loc=<-14.0474 77.9413 27.2458>, rot=<0.6805 0.175 0.6875 0.1836>}
 eyeE_mover {parent=eyeE, loc=<0 0 0>, rot=<0 0 0 1>}
 eyeE_orb {parent=eyeE_mover, loc=<279.1516 0 0>, rot=<0.002 0 0 -1>}
 eyeE_innerlid_mstr {parent=eyeE_orb, loc=<0 0 0>, rot=<0 0 0 1>}
 eyeE_innerlidB_mstr {parent=eyeE_innerlid_mstr, loc=<0 0 0>, rot=<0 0.2335 0.0081 -0.9723>}
 eyeE_innerlidB {parent=eyeE_innerlidB_mstr, loc=<499.6037 0 0>, rot=<0 0 0 1>}
 eyeE_innerlidA_mstr {parent=eyeE_innerlid_mstr, loc=<0 0 0>, rot=<0 0 0 1>}
 eyeE_innerlidA {parent=eyeE_innerlidA_mstr, loc=<512.2459 0.1408 2.3499>, rot=<-0.0003 0.002 -0.0112 -0.9999>}
 eyeE_ball {parent=eyeE_orb, loc=<0 0 0>, rot=<0 0 0 1>}
 eyeE_pupil {parent=eyeE_orb, loc=<750 0 0>, rot=<0 0 0 1>}
 eyeE_lid_upper_master {parent=eyeE_mover, loc=<0 0 0>, rot=<-0.0021 -0.0018 -0.1264 -0.992>}
 eyeE_lid_upperB {parent=eyeE_lid_upper_master, loc=<704.0695 3.7057 188.8306>, rot=<-0.1076 0.0474 -0.0618 -0.9911>}
 eyeE_lid_upperA {parent=eyeE_lid_upper_master, loc=<729.6466 5.3358 -185.2324>, rot=<0.1474 0.0032 -0.0296 -0.9886>}
 eyeE_lid_lower_master {parent=eyeE_mover, loc=<0 0 0>, rot=<0.9898 -0.1425 -0.0034 0.0021>}
 eyeE_lid_lowerB {parent=eyeE_lid_lower_master, loc=<722.881 13.011 -183.0044>, rot=<-0.0422 -0.9926 0.1127 0.0184>}
 eyeE_lid_lowerA {parent=eyeE_lid_lower_master, loc=<737.9814 -0.0785 184.7559>, rot=<-0.0643 -0.9902 -0.1238 0>}
 eyeD {parent=tentacle_08, loc=<792.9301 338.1037 -91.8744>, rot=<0.588 0.4008 0.5697 0.4111>}
 eyeD_mover {parent=eyeD, loc=<0 0 0>, rot=<0 0 0 1>}
 eyeD_orb {parent=eyeD_mover, loc=<0 -49.4335 0.0036>, rot=<-0.0436 0 0 -0.999>}
 eyeD_innerlid_mstr {parent=eyeD_orb, loc=<0 0 0>, rot=<0.0013 -0.0012 0.0125 -0.9999>}
 eyeD_innerlidA_mstr {parent=eyeD_innerlid_mstr, loc=<0 0 0>, rot=<0 0 0 1>}
 eyeD_innerlidA {parent=eyeD_innerlidA_mstr, loc=<782.9152 -8.2737 0.6861>, rot=<-0.0014 0.0056 0.0379 -0.9993>}
 eyeD_innerlidB_mstr {parent=eyeD_innerlid_mstr, loc=<0 0 0>, rot=<0 0.2271 0 -0.9739>}
 eyeD_innerlidB_adj {parent=eyeD_innerlidB_mstr, loc=<149.9942 -1.3209 0.0029>, rot=<0 0 0 1>}
 eyeD_innerlidB {parent=eyeD_innerlidB_adj, loc=<619.7855 -7.1135 2.6421>, rot=<0 0 0 1>}
 eyeD_ball {parent=eyeD_orb, loc=<0 0 0>, rot=<0 0 0 1>}
 eyeD_pupil {parent=eyeD_orb, loc=<1000 0 0>, rot=<0 0 0 1>}
 eyeD_lid_upper_master {parent=eyeD_mover, loc=<0 0 0>, rot=<-0.0035 -0.0105 -0.2028 -0.9792>}
 eyeD_lid_upperB {parent=eyeD_lid_upper_master, loc=<674.6255 -83.6545 297.4906>, rot=<-0.1025 0.0648 0.014 -0.9925>}
 eyeD_lid_upperA {parent=eyeD_lid_upper_master, loc=<738.3533 -97.4861 -269.6031>, rot=<0.148 0.0009 0.0483 -0.9878>}
 eyeD_lid_lower_master {parent=eyeD_mover, loc=<0 0 0>, rot=<0.9815 -0.1914 -0.0002 0.0054>}
 eyeD_lid_lowerB {parent=eyeD_lid_lower_master, loc=<726.6329 34.215 -188.2804>, rot=<-0.0917 -0.9898 0.1071 0.02>}
 eyeD_lid_lowerA {parent=eyeD_lid_lower_master, loc=<755.4763 10.9179 202.5951>, rot=<-0.1127 -0.9853 -0.1284 -0.01>}
 eyeC {parent=tentacle_06, loc=<283.0168 -272.6368 324.0001>, rot=<0.7084 -0.1764 0.6616 -0.1713>}
 eyeC_mover {parent=eyeC, loc=<0 0 0>, rot=<0 0 0 1>}
 eyeC_orb {parent=eyeC_mover, loc=<0 0 0>, rot=<-0.002 0 0 -1>}
 eyeC_innerlid_mstr {parent=eyeC_orb, loc=<0 0 0>, rot=<0 0 0 1>}
 eyeC_innerlidA_mstr {parent=eyeC_innerlid_mstr, loc=<0 0 0>, rot=<0 0 0 1>}
 eyeC_innerlidA {parent=eyeC_innerlidA_mstr, loc=<791.9602 0.1806 1.1381>, rot=<0 0 0 1>}
 eyeC_innerlidB_mstr {parent=eyeC_innerlid_mstr, loc=<0 0 0>, rot=<0.0059 0.2135 0.0162 -0.9768>}
 eyeC_innerlidB {parent=eyeC_innerlidB_mstr, loc=<772.1907 0 0>, rot=<-0.0087 -0.0406 -0.015 -0.999>}
 eyeC_ball {parent=eyeC_orb, loc=<0 0 0>, rot=<0 0 0 1>}
 eyeC_pupil {parent=eyeC_orb, loc=<1000 0 0>, rot=<0 0 0 1>}
 eyeC_lid_upper_master {parent=eyeC_mover, loc=<0 0 0>, rot=<0.9799 -0.1988 0.0167 0.0068>}
 eyeC_lid_upperB {parent=eyeC_lid_upper_master, loc=<724.6735 -86.8889 -234.918>, rot=<-0.1012 -0.9893 0.1051 0.0034>}
 eyeC_lid_upperA {parent=eyeC_lid_upper_master, loc=<706.252 -104.9976 255.9608>, rot=<-0.118 -0.984 -0.1304 -0.0283>}
 eyeC_lid_lower_master {parent=eyeC_mover, loc=<0 0 0>, rot=<-0.0066 0.0184 -0.182 -0.9831>}
 eyeC_lid_lowerB {parent=eyeC_lid_lower_master, loc=<709.759 -106.3475 243.1054>, rot=<-0.1092 0.0303 0.0336 -0.993>}
 eyeC_lid_lowerA {parent=eyeC_lid_lower_master, loc=<703.5789 -93.1802 -243.5871>, rot=<0.1455 -0.0187 0.0604 -0.9873>}
 eyeB {parent=tentacle_05, loc=<-43.2429 -653.2473 24.393>, rot=<0.5989 -0.3999 0.5675 -0.3991>}
 eyeB_mover {parent=eyeB, loc=<0 0 0>, rot=<0 0 0 1>}
 eyeB_orb {parent=eyeB_mover, loc=<0 73.6008 -0.3782>, rot=<0.0026 0.0001 0.0538 -0.9985>}
 eyeB_innerlid_mstr {parent=eyeB_orb, loc=<0 0 0>, rot=<0 0 0 1>}
 eyeB_innerlidB_mstr {parent=eyeB_innerlid_mstr, loc=<0 0 0>, rot=<-0.009 0.2171 -0.0314 -0.9756>}
 eyeB_innerlidB_adj {parent=eyeB_innerlidB_mstr, loc=<150 0 0>, rot=<0 0 0 1>}
 eyeB_innerlidB {parent=eyeB_innerlidB_adj, loc=<650.024 0 0>, rot=<0.0144 -0.0444 0.0294 -0.9985>}
 eyeB_innerlidA_mstr {parent=eyeB_innerlid_mstr, loc=<0 0 0>, rot=<0 0 0 1>}
 eyeB_innerlidA {parent=eyeB_innerlidA_mstr, loc=<795.264 12.1153 1.1381>, rot=<0 0 0 1>}
 eyeB_ball {parent=eyeB_orb, loc=<0 0 0>, rot=<0 0 0 1>}
 eyeB_pupil {parent=eyeB_orb, loc=<1000 0 0>, rot=<0 0 0 1>}
 eyeB_lid_upper_master {parent=eyeB_mover, loc=<0 -37.2459 0.1914>, rot=<0.9816 -0.1909 0.0016 0.0054>}
 eyeB_lid_upperB {parent=eyeB_lid_upper_master, loc=<735.3018 -33.838 -268.9658>, rot=<-0.1345 -0.9847 0.1086 0.0239>}
 eyeB_lid_upperA {parent=eyeB_lid_upper_master, loc=<727.8824 -23.2693 358.4098>, rot=<-0.1551 -0.9798 -0.1254 -0.0163>}
 eyeB_lid_lower_master {parent=eyeB_mover, loc=<0 -37.2459 0.1914>, rot=<-0.0074 0.0008 -0.2692 -0.9631>}
 eyeB_lid_lowerB {parent=eyeB_lid_lower_master, loc=<746.8317 13.0632 287.4309>, rot=<-0.0973 0.0542 0.0397 -0.993>}
 eyeB_lid_lowerA {parent=eyeB_lid_lower_master, loc=<712.019 3.8854 -323.6267>, rot=<0.1519 -0.0157 0.0708 -0.9857>}
 eyeA {parent=tentacle_03, loc=<-121.0337 656.824 180.5193>, rot=<0.5892 0.4265 0.5465 0.4151>}
 eyeA_mover {parent=eyeA, loc=<0 0 0>, rot=<0 0 0 1>}
 eyeA_orb {parent=eyeA_mover, loc=<0 0 0>, rot=<0.0015 0 0 -1>}
 eyeA_innerlid_mstr {parent=eyeA_orb, loc=<0 0 0>, rot=<0 0 0 1>}
 eyeA_innerlidB_mstr {parent=eyeA_innerlid_mstr, loc=<0 0 0>, rot=<0.0028 0.2037 0.0189 -0.9789>}
 eyeA_innerlidB_adj {parent=eyeA_innerlidB_mstr, loc=<150 0 0>, rot=<0 0 0 1>}
 eyeA_innerlidB {parent=eyeA_innerlidB_adj, loc=<658.3829 0 0>, rot=<0 0 0 1>}
 eyeA_innerlidA_mstr {parent=eyeA_innerlid_mstr, loc=<0 0 0>, rot=<0 0 0 1>}
 eyeA_innerlidA {parent=eyeA_innerlidA_mstr, loc=<791.9602 0.1806 1.1381>, rot=<0 0 0 1>}
 eyeA_ball {parent=eyeA_orb, loc=<0 0 0>, rot=<0 0 0 1>}
 eyeA_pupil {parent=eyeA_orb, loc=<1000 0 0>, rot=<0 0 0 1>}
 eyeA_lid_upper_master {parent=eyeA_mover, loc=<0 0 0>, rot=<0.0004 -0.0238 -0.3083 -0.951>}
 eyeA_lid_upperB {parent=eyeA_lid_upper_master, loc=<687.9968 -77.1336 368.4009>, rot=<-0.0959 0.0892 0.1207 -0.984>}
 eyeA_lid_upperA {parent=eyeA_lid_upper_master, loc=<749.6166 -99.2781 -334.1253>, rot=<0.1457 -0.0017 0.1577 -0.9767>}
 eyeA_lid_lower_master {parent=eyeA_mover, loc=<0 0 0>, rot=<0.9699 -0.2425 -0.0214 0.0004>}
 eyeA_lid_lowerB {parent=eyeA_lid_lower_master, loc=<713.6677 5.1119 -376.208>, rot=<-0.1408 -0.9827 0.11 0.0478>}
 eyeA_lid_lowerA {parent=eyeA_lid_lower_master, loc=<772.2507 -25.9781 373.8751>, rot=<-0.1669 -0.9783 -0.123 0.0054>}
 ik_tentacle {parent=root, loc=<1.0934 -210.2722 12975.9268>, rot=<-0.4875 0.4878 -0.5122 -0.5119>}
 Eye_X_Offset {type=Curve, loc=<0 0 0>, rot=<0 0 0 1>}
 Eye_Y_Offset {type=Curve, loc=<0 0 0>, rot=<0 0 0 1>}
 ik_tentacle_ctrl {controlType=EULER_TRANSFORM, type=Control, loc=<0 0 0>, rot=<0 0 0 1>}
]
variables [
 DoIKTrace {type=bool, default=False}
 ZOffset {type=double, default=0}
 ZOffset_Target {type=double, default=0}
]
