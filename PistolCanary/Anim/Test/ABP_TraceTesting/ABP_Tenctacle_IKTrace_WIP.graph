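# Work-in-progress test AnimBlueprint for the PistolCanary creature tentacle, pairing a ground slam attack animation with Control Rig IK tracing.
# A single-state Slam state machine plays the GroundSlam sequence and feeds the IKTrace Control Rig, which outputs the final pose.
# The event graph holds only unconnected default nodes, and the BP_Tentacle variable is not used in any graph.
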AnimBlueprint /PistolCanary/Anim/Test/ABP_TraceTesting/ABP_Tenctacle_IKTrace_WIP
target_skeleton=/PistolCanary/Assets/Meshes/Tentacle_PistolCanary_Creature_Skeleton
graphs [
 AnimGraph {nodes [
   AnimGraphNode_StateMachine_0 = Slam() {is=title AnimGraphNode_StateMachine}
   n1 = Control Rig(Source=@AnimGraphNode_StateMachine_0) {
    alphaBoolBlend {blendOption=Linear}
    controlRigAssetReference {blueprintRigClass=&/Tentacle_PistolCanary_Creature_IKTrace_CtrlRig.Tentacle_PistolCanary_Creature_IKTrace_CtrlRig_C}
   }
   Output Pose(Result=@n1)
  ]}
 Slam {entry=Slam, path=AnimGraph.AnimGraphNode_StateMachine_0.Slam, ids=id, nodes [AnimStateNode_0 = Slam() {is=title AnimStateNode}]}
 Slam {
  path=AnimGraph.AnimGraphNode_StateMachine_0.Slam.AnimStateNode_0.Slam
  nodes [Output Animation Pose(Result=Sequence Player 'PistolCanary_Tentacle_Attack_GroundSlam'() {sequence=/PistolCanary/Anim/SingleTentacle/Section02/PistolCanary_Attack_GroundSlam/PistolCanary_Tentacle_Attack_GroundSlam})]
 }
 EventGraph {nodes [
   TryGetPawnOwner()
   Event BlueprintUpdateAnimation()
  ]}
]
variables [BP_Tentacle {type=BP_Tentacle_C, default=null}]
